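# Niagara system for the Cavalry Alt pickaxe's weapon trail, built from a single ribbon emitter, Trail_Noise.
# Short-lived ribbon particles spawn continuously at the axe mesh's FX_Trail_1 socket through skeletal mesh sampling.
# Over each particle's life the RGB quickly darkens to black while alpha fades and width tapers, with time-of-day scalars and tints adjusting the color.
# Dynamic material parameters pass life-based fade curves and a random value to the ribbon material.
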NiagaraSystem /BRCosmetics/Weapons/FORT_Melee/Pickaxe_Cavalry_Alt/FX/NS_Pickaxe_Cavalry_Alt_Trail
dependencies {
 usedDataInterfaces [&NiagaraDataInterfaceColorCurve, &NiagaraDataInterfaceCurve]
 usedDynamicInputs []
 usedModules [&SystemState]
}
emitters {Trail_Noise {
  emitterUpdateScript [
   &EmitterState()
   &FN_SpawnRibbon_v2(SpawnRate=80)
  ]
  particleSpawnScript [
   &InitializeRibbon(Color=<rgba 5.5 5.5 5.5 1>, Lifetime=0.17, Ribbon Width=12, UsePositionOffset=false, Write Material Random=true)
   &SkeletalMeshLocation(
    MaskSpawnGroup=false
    Mesh Sampling Type=::NewEnumerator5
    OverrideRandomSeed=false
    Skeletal Mesh=DataInterface() {propertyValues {
      DefaultMesh=null
      ExcludeBoneName=root
      OptionalComponentClass=null
      PreviewMesh=/BRCosmetics/Weapons/FORT_Melee/Pickaxe_Cavalry_Alt/Meshes/Cavalry_Alt_Axe
      SkinningMode=SkinOnTheFly
      SoftSourceActor=null
      SourceMode=Default
      UvSetIndex=0
      WholeMeshLOD=-1
      bExcludeBone=false
      bReadDeformedGeometry=true
      bRequireCurrentFrameData=true
      ComponentTags []
      FilteredBones []
      FilteredSockets [FX_Trail_1]
      MeshUserParameter {parameter {name=None, typeDefHandle {registeredTypeIndex=1}, varData []}}
      SamplingRegions []
     }}
   )
  ]
  particleUpdateScript [
   &ParticleState()
   &ScaleColor(
    Linear Color Curve=DataInterface() {propertyValues {
      AlphaCurve {keys [<at 0.0066 1>,<at 0.3319 1>,<at 0.9833 0>]}
      BlueCurve {keys [<at 0 0.1>,<at 0.1274 0.025>,<at 0.2123 0>]}
      GreenCurve {keys [<at 0 0.1>,<at 0.1274 0.025>,<at 0.2123 0>]}
      RedCurve {keys [<at 0 0.1>,<at 0.1274 0.025>,<at 0.2123 0>]}
     }}
    Scale Mode=::NewEnumerator2
   )
   &ScaleColor(
    Scale Alpha=&FloatFromCurve(
     FloatCurve=DataInterface() {propertyValues {Curve {keys [
         {interpMode=RCIM_Cubic, tangentMode=RCTM_User, time=0, value=1}
         {arriveTangent=-0.4716, interpMode=RCIM_Cubic, leaveTangent=-0.4716, tangentMode=RCTM_User, time=1, value=0}
        ]}}}
     Scale Curve=&Multiply_Float(A=&FNTimeOfDayScalar(Night Scale=1.1, Morning Scale=1.1, Day Scale=1.1, Evening Scale=1.1), B=1)
    )
    Scale RGB=&FN_Select_Vector_from_Time_Of_Day(
     Morning=&MakeVectorFromLinearColorRGB(Color=<rgba 0.625 0.4516 0.2688 1>)
     Day=&MakeVectorFromLinearColorRGB(Color=<rgba 0.7969 0.7969 0.7969 1>)
     Evening=&MakeVectorFromLinearColorRGB(Color=<rgba 1 1 1 1>)
     Night=&MakeVectorFromLinearColorRGB(Color=<rgba 0.5618 0.6823 0.5384 1>)
    )
    ScaleA=true
    ScaleRGB=true
   ) {module=ScaleColor001}
   &ScaleRibbonWidth(Ribbon Width Scale=&FloatFromCurve(FloatCurve=DataInterface() {propertyValues {Curve {keys [
         {arriveTangent=-0.0008, interpMode=RCIM_Cubic, leaveTangent=-0.0008, tangentMode=RCTM_User, time=0, value=1}
         {arriveTangent=-1.193, interpMode=RCIM_Cubic, leaveTangent=-1.193, tangentMode=RCTM_User, time=1, value=0.5}
        ]}}}))
   &DynamicMaterialParameters(
    Index 0 Param 1=&FloatFromCurve(FloatCurve=DataInterface() {propertyValues {Curve {keys [<at 0 1>,<at 1 0>]}}})
    Index 0 Param 2=&RandomRangeFloat(Minimum=0, Maximum=1)
    Index 0 Param 4=&FloatFromCurve(FloatCurve=DataInterface() {propertyValues {Curve {keys [
         {interpMode=RCIM_Cubic, tangentMode=RCTM_User, time=0, value=1}
         {arriveTangent=-0.6219, interpMode=RCIM_Cubic, leaveTangent=-0.6219, tangentMode=RCTM_User, time=1, value=0}
        ]}}})
    Module.Param0WriteEnabled=true
    Module.Param1WriteEnabled=true
    Module.Param2WriteEnabled=true
    Module.Param3WriteEnabled=true
   )
  ]
  rendererClasses [&NiagaraRibbonRendererProperties]
  renderers [{rendererClass=&NiagaraRibbonRendererProperties}]
 }}
summary {
 systemName=NS_Pickaxe_Cavalry_Alt_Trail
 emitters [{emitterName=Trail_Noise, rendererClasses [&NiagaraRibbonRendererProperties]}]
 userVariables [
  User.RibbonIDOffset {type=NiagaraInt32}
  User.SkeletalMesh {
   type=NiagaraDataInterfaceSkeletalMesh
   defaultValue {
    struct=NiagaraExt_VariableValue_DataInterface
    value {dataInterface=&self:NiagaraDataInterfaceSkeletalMesh_0}
   }
  }
  User.TrailColor {type=LinearColor, description="", defaultValue {struct=LinearColor, value=<rgba 1 1 1 0.35>}}
 ]
}
